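; Music encounter Blueprint for the Stalker Knife that plays adaptive music around a targeted player pawn.
; On start it spawns 2D music, resets the MusicEncounter.Phase parameter and attaches the audio component to the target's root.
; Kill and unmark events switch that phase, with unmarking also pausing the encounter once audio finishes.
; Functions set the target pawn and feed DistanceToStalker into the playing audio as a float parameter.
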(event EventInitializeMusicEncounter
  (|CallParentFunction)
  (bind _asb_stalker_knife_mecomponent (Utilities|Casting|CastToB_StalkerKnife_MEComponent (Utilities|Array|Get(acopy) (MusicEncounter|GetMusicEncounterComponents))))
  (Variables|Default|SetAs%20B%20Stalker%20Knife%20MEComponent _asb_stalker_knife_mecomponent))

(event EventStartMusicEncounter
  (|CallParentFunction)
  (Variables|Default|GetPlayingMusicComponent
    (:then)
    (:else
      (bind _musictoplay (Variables|Default|GetMusicToPlay)
        (:then
          (bind _audio_component (MusicEncounter|SpawnMusic2D _musictoplay))
          (bind _output_get (Variables|Default|SetPlayingMusicComponent _audio_component))
          (Audio|Parameter|SetIntegerParameter _output_get "MusicEncounter.Phase" 0 :!class /Script/AudioExtensions.AudioParameterControllerInterface)
          (bind _targetedplayerpawn (Variables|Default|GetTargetedPlayerPawn)
            (:then
              (Transformation|AttachComponentToComponent (Variables|Default|GetPlayingMusicComponent) (Variables|/Script/Engine.Actor|GetRootComponent _targetedplayerpawn) "None" "KeepRelative" "KeepRelative" "KeepRelative" true))
            (:else
              (Development|PrintString 0))))
        (:else)))))

(event Custom|ME_TargetNoLongerMarked
  (Variables|Default|SetCurrentPhase 2)
  (bind _playingmusiccomponent (Variables|Default|GetPlayingMusicComponent)
    (:then
      (EventDispatchers|AssignOnAudioFinished _playingmusiccomponent (delegate Custom|OnAudioFinished_Event))
      (Audio|Parameter|SetIntegerParameter (stale Custom|ME_TargetKilled _playingmusiccomponent) "MusicEncounter.Phase" (Variables|Default|GetCurrentPhase) :!class /Script/AudioExtensions.AudioParameterControllerInterface))
    (:else)))

(event Custom|OnAudioFinished_Event
  (CallFunction|PauseMusicEncounter))

(event Custom|ME_TargetKilled
  (Variables|Default|SetCurrentPhase 2)
  (Variables|Default|SetbIsTargetOrStalkerDefeated true)
  (bind _playingmusiccomponent (Variables|Default|GetPlayingMusicComponent)
    (:then
      (Audio|Parameter|SetIntegerParameter _playingmusiccomponent "MusicEncounter.Phase" (Variables|Default|GetCurrentPhase) :!class /Script/AudioExtensions.AudioParameterControllerInterface))
    (:else)))

(event Custom|ME_StalkerKilled
  (Variables|Default|SetCurrentPhase 1)
  (Variables|Default|SetbIsTargetOrStalkerDefeated true)
  (bind _playingmusiccomponent (Variables|Default|GetPlayingMusicComponent)
    (:then
      (Audio|Parameter|SetIntegerParameter _playingmusiccomponent "MusicEncounter.Phase" (Variables|Default|GetCurrentPhase) :!class /Script/AudioExtensions.AudioParameterControllerInterface))
    (:else)))

(fn SetTargetedPlayerPawn
  (params (TargetedPlayerPawn :type Fort%20Pawn%20Object%20Reference))
  (Variables|Default|SetTargetedPlayerPawn TargetedPlayerPawn))

(fn UpdateDistanceToStalker
  (params (InFloat :type Float%20%28single-precision%29))
  (bind _playingmusiccomponent (Variables|Default|GetPlayingMusicComponent)
    (:then
      (Audio|Parameter|SetFloatParameter _playingmusiccomponent "DistanceToStalker" InFloat :!class /Script/AudioExtensions.AudioParameterControllerInterface))
    (:else)))
