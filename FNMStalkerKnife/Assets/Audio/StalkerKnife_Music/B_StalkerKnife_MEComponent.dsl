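; Music encounter component that drives Stalker Knife music on the marked player pawn from the nearest stalker's distance.
; Stalkers live in a map of pawn to expiry time; tick prunes them and recomputes the closest, using plain distance despite "DistSq" names.
; A looping timer smooths that distance with FInterpTo and feeds B_StalkerKnife_ME.
; It also forwards kill and mark-removal events; DBNO binding only prints placeholder text.
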(event EventBeginPlay
  (|CallParentFunction))

(event EventInitializeComponent
  (|CallParentFunction)
  (Networking|CanExecuteCosmeticEvents
    (:True
      (bind _asb_stalker_knife_me (Utilities|Casting|CastToB_StalkerKnife_ME (MusicEncounterComponent|GetMusicEncounter))
        (:then
          (Variables|Default|SetAs%20B%20Stalker%20Knife%20ME _asb_stalker_knife_me)
          (bind _asfort_player_pawn (Utilities|Casting|CastToFortPlayerPawn (Components|GetOwner))
            (:then
              (bind _output_get (Variables|Default|SetTargetedPlayerPawn _asfort_player_pawn))
              (Ability|Tasks|WaitGameplayTagRemoveFromActor _output_get (Variables|Default|GetTargetMarkedGameplayTag)
                (:then
                  (Class|BStalkerKnifeME|SetTargetedPlayerPawn (Variables|Default|GetAs%20B%20Stalker%20Knife%20ME) (Variables|Default|GetTargetedPlayerPawn)))
                (:Removed
                  (CallFunction|MECStopDistanceUpdateTimer)
                  (Default|ClearStalkers)
                  (bind _as_b_stalker_knife_me (Variables|Default|GetAs%20B%20Stalker%20Knife%20ME)
                    (:then
                      (Class|BStalkerKnifeME|METargetNoLongerMarked _as_b_stalker_knife_me))
                    (:else)))))
            (:CastFailed
              (Development|PrintString "Hello"))))
        (:CastFailed
          (Development|PrintString "Hello"))))))

(event EventTick
  (params (DeltaSeconds :type Float%20%28single-precision%29))
  ;[0.6574 0.1718 0.8594 0.5] Default tick is around 0.5 from parent class
  (|CallParentFunction DeltaSeconds)
  (CallFunction|TickStalkerTracking))

(event EventEndPlay
  (params (EndPlayReason :type EEndPlayReason%20Enum :default "Destroyed"))
  (|CallParentFunction EndPlayReason)
  (Utilities|Time|ClearandInvalidateTimerbyHandle (Variables|Default|GetDistanceTimerHandle)))

(event Custom|MEC_StopDistanceUpdateTimer
  (Utilities|Time|ClearandInvalidateTimerbyHandle (Variables|Default|GetDistanceTimerHandle)))

(event Custom|MEC_StartDistanceUpdateTimer
  (if (Utilities|Time|IsTimerActivebyHandle (Variables|Default|GetDistanceTimerHandle))
    (Utilities|Time|ClearandInvalidateTimerbyHandle (Variables|Default|GetDistanceTimerHandle)))
  (bind _returnvalue (Utilities|Time|SetTimerbyEvent (delegate Custom|UpdateMusicDistanceTimer) (Variables|Default|GetUpdateTime) true true))
  (Variables|Default|SetDistanceTimerHandle _returnvalue)
  (bind _as_b_stalker_knife_me (Variables|Default|GetAs%20B%20Stalker%20Knife%20ME)
    (:then
      (Class|MusicEncounterBase|StartMusicEncounter _as_b_stalker_knife_me))
    (:else)))

(event Custom|UpdateMusicDistanceTimer
  (bind _output_get (Variables|Default|SetClosestDistSq_Smoothed (Math|Interpolation|FInterpTo (Variables|Default|GetClosestDistSq_Smoothed) (Variables|Default|GetClosestDistSq) (Variables|Default|GetUpdateTime) (Variables|Default|GetInterp%20Speed))))
  (Development|PrintString (Utilities|String|ToString(Float) _output_get) (Variables|Default|GetbDebugText) false "(R=0.000000,G=0.660000,B=1.000000,A=1.000000)" 2.0 "StalkerKnife_Key")
  (CallFunction|CheckforInvalidStalkers)
  (CallFunction|UpdateMusicDistance))

(event Custom|MEC_TargetKilled
  (params (TargetPawn :type Fort%20Pawn%20Object%20Reference))
  (if (Math|Boolean|ANDBoolean (Utilities|IsValid TargetPawn :!class /Script/Engine.KismetSystemLibrary) (== TargetPawn (Variables|Default|GetTargetedPlayerPawn)) (Utilities|IsValid (Variables|Default|GetTargetedPlayerPawn) :!class /Script/Engine.KismetSystemLibrary))
    (bind _as_b_stalker_knife_me (Variables|Default|GetAs%20B%20Stalker%20Knife%20ME)
      (:then
        (Class|BStalkerKnifeME|METargetKilled _as_b_stalker_knife_me))
      (:else))))

(event Custom|StalkerNoLongerValid
  (bind _as_b_stalker_knife_me (Variables|Default|GetAs%20B%20Stalker%20Knife%20ME)
    (:then
      (Class|BStalkerKnifeME|MEStalkerKilled _as_b_stalker_knife_me))
    (:else)))

(event Custom|UpdateBinding
  (bind _keypin (Utilities|Map|ForEachLoop(Map) (Variables|Default|GetActiveStalkers)))
  (Ability|Tasks|WaitGameplayTagRemoveFromActor _keypin "(TagName=\"Gameplay.Status.DBNO\")"
    (:then
      (Ability|Tasks|WaitGameplayTagAddToActor _keypin "(TagName=\"Gameplay.Status.DBNO\")"
        (:Added
          (Development|PrintString "Hello"))))
    (:Removed
      (Development|PrintString "Hello"))))

(fn AddOrRefreshStalker
  (params (NewStalkerPawn :type Fort%20Pawn%20Object%20Reference) (Duration :type Float%20%28double-precision%29))
  (Utilities|IsValid NewStalkerPawn
    (:Is%20Valid
      (Utilities|Map|Add (Variables|Default|GetActiveStalkers) NewStalkerPawn (+ (Math|Float|Max(Float) Duration) (Utilities|Time|GetGameTimeinSeconds)))
      (CallFunction|MECStartDistanceUpdateTimer)
      (CallFunction|UpdateBinding))))

(fn CheckForInvalidStalkers
  (bind _activestalkers (Variables|Default|GetActiveStalkers))
  (if (Utilities|Map|IsNotEmpty _activestalkers)
    (bind _keypin (Utilities|Map|ForEachLoop(Map) _activestalkers))
    (Utilities|IsValid _keypin
      (:Is%20Not%20Valid
        (CallFunction|RemoveStalker _keypin)
        ;[0.5392 0.2001 0.6615 0.5] play stalker killed stinger
        (bind _as_b_stalker_knife_me (Variables|Default|GetAs%20B%20Stalker%20Knife%20ME)
          (:then
            (Class|BStalkerKnifeME|MEStalkerKilled _as_b_stalker_knife_me))
          (:else))))
    (else
      (CallFunction|MECStopDistanceUpdateTimer)
      (bind _as_b_stalker_knife_me_2 (Variables|Default|GetAs%20B%20Stalker%20Knife%20ME)
        (:then
          (Class|BStalkerKnifeME|METargetNoLongerMarked _as_b_stalker_knife_me_2)
          (Default|ClearStalkers))
        (:else)))))

(fn ClearStalkers
  (Utilities|Map|Clear (Variables|Default|GetActiveStalkers))
  (CallFunction|RecomputeClosest))

(fn HasDistanceChangedEnough
  (params (BestDistanceSq :type Float%20%28double-precision%29) (BestStalker :type Object%20Reference))
  (return (or (!= BestStalker (Variables|Default|GetClosestStalker)) (> (Math|Float|Absolute(Float) (- BestDistanceSq (Variables|Default|GetClosestDistSq))) (Variables|Default|GetMinDistanceChangeSq)))))

(fn RecomputeClosest
  (bind _targetedplayerpawn (Variables|Default|GetTargetedPlayerPawn))
  (Variables|Default|GetTargetedPlayerPawn
    (:then
      (if (Utilities|Map|IsEmpty (Variables|Default|GetActiveStalkers))
        (Development|PrintString (Utilities|String|ToString(Text) (Utilities|Text|FormatText "NSLOCTEXT(\"[D7FA64BFC880001C8DC4183E3FE16DE4]\", \"E9AE4883461DEC25521AC29662AAD10D\", \"Target: {t}\\r\\nClosest Stalker: {s}\")" (Variables|Default|GetTargetedPlayerPawn) (Variables|Default|GetClosestStalker))) (Variables|Default|GetbDebugText) false "(R=1.000000,G=0.826823,B=0.062664,A=1.000000)" 2.0 "StalkerKnife_2")
        (return)
        (elif (!= (Utilities|Map|Length (Variables|Default|GetActiveStalkers)) 1)
          (bind _keypin (Utilities|Map|ForEachLoop(Map) (Variables|Default|GetActiveStalkers))
            (:then
              (Utilities|IsValid _keypin
                (:Is%20Valid
                  (if (< (Transformation|GetDistanceTo _targetedplayerpawn _keypin) (LocalVariables|GetBestDistSq))
                    (LocalVariables|SetBestDistSq (Transformation|GetDistanceTo _targetedplayerpawn _keypin))
                    (LocalVariables|SetBestStalker _keypin)))))
            (:CompletedPin
              (if (CallFunction|HasDistanceChangedEnough (LocalVariables|GetBestDistSq) (LocalVariables|GetBestStalker))
                (Variables|Default|SetClosestStalker (LocalVariables|GetBestStalker))
                (bind _output_get (Variables|Default|SetClosestDistSq (LocalVariables|GetBestDistSq)))
                (Development|PrintString (Utilities|String|ToString(Text) (Utilities|Text|FormatText "NSLOCTEXT(\"[D7FA64BFC880001C8DC4183E3FE16DE4]\", \"896D86EB4D5D918AD5AB08852F29C05F\", \"Closest Stalker: {s}\\r\\nDistance: {d}\")" (Variables|Default|GetClosestStalker) _output_get)) (Variables|Default|GetbDebugText) false "(R=0.123799,G=0.225650,B=1.000000,A=1.000000)" 2.0 "StalkerKnife_3"))))
          (else
            (bind _keys (Utilities|Map|Keys (Variables|Default|GetActiveStalkers)))
            (bind _output_get_2 (Variables|Default|SetClosestStalker (Utilities|Array|Get(acopy) _keys)))
            (LocalVariables|SetBestDistSq (Transformation|GetDistanceTo (Variables|Default|GetTargetedPlayerPawn) _output_get_2))))))
    (:else
      (Development|PrintString "Stalker Knife Target Player NOT Valid Yet" (Variables|Default|GetbDebugText) true "(R=1.000000,G=0.000000,B=0.000000,A=1.000000)" 2.0 "Stalker Knife Target Player NOT Valid Yet"))))

(fn RemoveStalker
  (params (StalkerToRemove :type Fort%20Pawn%20Object%20Reference))
  (Utilities|Map|Remove (Variables|Default|GetActiveStalkers) StalkerToRemove))

(fn StalkerKnife_InterpSpeed
  (params (Interp%20Speed :type Float%20%28double-precision%29))
  (Variables|Default|SetInterp%20Speed Interp%20Speed))

(fn TickStalkerTracking
  (bind _activestalkers (Variables|Default|GetActiveStalkers))
  (LocalVariables|SetTime_Now (Utilities|Time|GetGameTimeinSeconds))
  (if (Utilities|Map|IsNotEmpty _activestalkers)
    (bind (_keypin _valuepin) (Utilities|Map|ForEachLoop(Map) _activestalkers)
      (:then
        (if (or (not (Utilities|IsValid _keypin :!class /Script/Engine.KismetSystemLibrary)) (< _valuepin (LocalVariables|GetTime_Now)))
          (CallFunction|RemoveStalker _keypin)))
      (:CompletedPin
        (CallFunction|RecomputeClosest)))))

(fn UpdateMusicDistance
  (bind _as_b_stalker_knife_me (Variables|Default|GetAs%20B%20Stalker%20Knife%20ME)
    (:then
      (Class|BStalkerKnifeME|UpdateDistancetoStalker _as_b_stalker_knife_me (Variables|Default|GetClosestDistSq_Smoothed)))
    (:else)))
