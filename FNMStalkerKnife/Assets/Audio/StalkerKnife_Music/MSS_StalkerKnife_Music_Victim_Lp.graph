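# Stereo MetaSound source for the Stalker Knife victim music, driven by encounter phase, tempo and distance to the stalker.
# Phase changes select a wave from an array, started on a tempo-derived trigger grid with per-phase gain.
# A second looping layer starts at a random offset in phase zero.
# Distance shapes volume through a wavetable curve, and the last phase releases the envelope that ends the sound.
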MetaSoundSource /FNMStalkerKnife/Assets/Audio/StalkerKnife_Music/MSS_StalkerKnife_Music_Victim_Lp
inputs [
 DistanceToStalker {type=Float}
 MusicEncounter.bIsPaused {type=Bool}
 MusicEncounter.PauseFadeOutTime {type=Time}
 MusicEncounter.Phase {type=Int32}
 Tempo {type=Float}
 UE.Source.OnPlay {type=Trigger}
]
interfaces [UE.OutputFormat.Stereo:1.0, UE.Source.OneShot:1.0, UE.Source:1.1]
outputs [
 "UE.OutputFormat.Stereo.Audio:0" {type=Audio}
 "UE.OutputFormat.Stereo.Audio:1" {type=Audio}
 UE.Source.OneShot.OnFinished {type=Trigger}
]
pages [Default {
  input_defaults {
   DistanceToStalker=827.9983
   MusicEncounter.PauseFadeOutTime=3
   MusicEncounter.Phase=0
   MusicEncounter.bIsPaused=null
   Tempo=120
   UE.Source.OnPlay=null
  }
  nodes [
   InitVariable.MIDIClock()
   InitVariable.WaveAsset()
   InitVariable.WaveAsset()
   n1 = VariableAccessor.WaveAsset:Array(Variable=InitVariable.WaveAsset:Array().Variable)
   n2 = UE.Subtract:Int32(PrimaryOperand=Array.Num:WaveAsset:Array(Array=VariableAccessor.WaveAsset:Array(Variable=@n1.Variable).Value).Num) {also {defaults {AdditionalOperands=1}}}
   n3 = InitVariable.Trigger()
   TriggerOnValueChange_1 = UE.Trigger On Value Change:Int32(Value=^MusicEncounter.Phase)
   TriggerAnyTriggerAny2_0 = "TriggerAny.Trigger Any (2)"(In 1=^UE.Source.OnPlay, In 0=@TriggerOnValueChange_1.Trigger)
   TriggerOnce_1 = UE.Trigger Once(
    Trigger In=UE.TriggerRepeat(Start=^UE.Source.OnPlay, Period=UE.BPMToSeconds(BPM=^Tempo) {also {defaults {Divisions of Whole Note=8}}}.Seconds).RepeatOut
    Reset=VariableAccessor.Trigger(Variable=VariableMutator.Trigger(Variable=@n3.Variable, Value=@TriggerAnyTriggerAny2_0.Out).Variable).Value
   ) {also {defaults {Start Closed="<MetaSoundBuilderResult.SUCCEEDED: 0>"}}}
   TriggerDelay_0 = UE.Trigger Delay(Delay Time=UE.BPMToSeconds(BPM=^Tempo).Seconds, In=@TriggerOnce_1.Trigger Out)
   TriggerOnce_3 = UE.Trigger Once(Reset=@TriggerDelay_0.Out, Trigger In=@TriggerOnce_1.Trigger Out)
   n4 = TriggerCompare.Clamp:Int32(A=^MusicEncounter.Phase, Compare=UE.Reroute(Value=@TriggerOnce_3.Trigger Out).Value) {also {defaults {B=0}}}
   ADSREnvelopeADSREnvelope_0 = ADSR Envelope.ADSR Envelope:Audio(
    Trigger Attack=^UE.Source.OnPlay
    Trigger Release=UE.Reroute(Value=TriggerCompare.Clamp:Int32(A=^MusicEncounter.Phase, B=@n2.Out, Compare=@n4.False).True).Value
   ) {also {defaults {Attack Time=6, Sustain Level=1}}}
   DecibelstoLinearGain_0 = UE.Decibels to Linear Gain:Float() {also {defaults {Decibels=3}}}
   DecibelstoLinearGain_1 = UE.Decibels to Linear Gain:Float()
   DecibelstoLinearGain_2 = UE.Decibels to Linear Gain:Float()
   DecibelstoLinearGain_3 = UE.Decibels to Linear Gain:Float()
   DecibelstoLinearGain_4 = UE.Decibels to Linear Gain:Float() {also {defaults {Decibels=3}}}
   n5 = UE.Reroute(Value=@TriggerOnce_3.Trigger Out)
   n6 = VariableDeferredAccessor.Trigger(Variable=InitVariable.Trigger().Variable)
   TriggerOnce_2 = UE.Trigger Once(Reset=@n5.Value, Trigger In=@n6.Value)
   WavePlayer_1 = UE.Wave Player:Stereo(
    Wave Asset=Array.Get:WaveAsset:Array(Index=^MusicEncounter.Phase, Trigger=@n5.Value, Array=@n1.Value).Element
    Play=@n5.Value
    Stop=@TriggerOnce_2.Trigger Out
   ) {also {version [1,1]}}
   TriggerSelect4_0 = "UE.Trigger Select (4)"(Index=^MusicEncounter.Phase, In=@WavePlayer_1.On Play)
   TriggerRouteTriggerRouteFloat4_0 = "TriggerRoute.Trigger Route (Float, 4):Float"(
    Value 1=@DecibelstoLinearGain_1.Linear Gain
    Value 0=@DecibelstoLinearGain_2.Linear Gain
    Value 2=@DecibelstoLinearGain_3.Linear Gain
    Value 3=@DecibelstoLinearGain_4.Linear Gain
    Set 0=@TriggerSelect4_0.Out 0
    Set 1=@TriggerSelect4_0.Out 1
    Set 2=@TriggerSelect4_0.Out 2
    Set 3=@TriggerSelect4_0.Out 3
   )
   n7 = VariableAccessor.WaveAsset(Variable=InitVariable.WaveAsset().Variable)
   GetWaveDuration_0 = "UE.Get Wave Duration: "(Wave=@n7.Value)
   TriggerOnce_0 = UE.Trigger Once(Trigger In=@n4.True)
   n8 = UE.RandomTime(Max=@GetWaveDuration_0.Duration, Next=@TriggerOnce_0.Trigger Out)
   WavePlayer_0 = UE.Wave Player:Stereo(Play=@n8.On Next, Start Time=@n8.Value, Stop=VariableAccessor.Trigger(Variable=InitVariable.Trigger().Variable).Value, Wave Asset=@n7.Value) {also {defaults {Loop="<MetaSoundBuilderResult.SUCCEEDED: 0>"}, version [1,1]}}
   AudioMixerAudioMixerStereo2_0 = "AudioMixer.Audio Mixer (Stereo, 2)"(
    Gain 1=@DecibelstoLinearGain_0.Linear Gain
    Gain 0=@TriggerRouteTriggerRouteFloat4_0.Value
    In 1 L=@WavePlayer_0.Out Left
    In 1 R=@WavePlayer_0.Out Right
    In 0 L=@WavePlayer_1.Out Left
    In 0 R=@WavePlayer_1.Out Right
   )
   n9 = InitVariable.Float()
   n10 = InitVariable.Float()
   n11 = VariableMutator.Float(
    Variable=InitVariable.Float().Variable
    Value=MapRange.MapRange:Float(
     In=^DistanceToStalker
     In Range B=VariableAccessor.Float(Variable=@n9.Variable).Value
     In Range A=VariableAccessor.Float(Variable=@n10.Variable).Value
    ).Out Value
   )
   n12 = UE.WaveTableBankEvaluate(Input=UE.InterpTo:Audio(Target=VariableAccessor.Float(Variable=@n11.Variable).Value) {also {defaults {Interp Time=0.2}}}.Value) {also {defaults {Index=1, WaveTableBank=&/WTB_StalkerKnife_Music_Victim}}}
   Multiply_0 = UE.Multiply:Audio(
    AdditionalOperands=@ADSREnvelopeADSREnvelope_0.Out Envelope
    PrimaryOperand=UE.Multiply:Audio by Float(PrimaryOperand=@AudioMixerAudioMixerStereo2_0.Out R, AdditionalOperands=@n12.Output).Out
   )
   Multiply_1 = UE.Multiply:Audio(
    AdditionalOperands=@ADSREnvelopeADSREnvelope_0.Out Envelope
    PrimaryOperand=UE.Multiply:Audio by Float(PrimaryOperand=@AudioMixerAudioMixerStereo2_0.Out L, AdditionalOperands=@n12.Output).Out
   )
   TriggerOnValueChange_0 = UE.Trigger On Value Change:Bool(Value=^MusicEncounter.bIsPaused)
   VariableMutator.Trigger(
    Variable=InitVariable.Trigger().Variable
    Value=TriggerCompare.Clamp:Bool(A=^MusicEncounter.bIsPaused, Compare=@TriggerOnValueChange_0.Trigger) {also {defaults {B="<MetaSoundBuilderResult.SUCCEEDED: 0>"}}}.True
   )
   VariableMutator.Trigger(Variable=@n6.Variable, Value=@WavePlayer_1.On Finished)
  ]
  links [
   {dst="@UE.Source.OneShot.OnFinished", src=ADSREnvelopeADSREnvelope_0.On Done}
   {dst="@UE.OutputFormat.Stereo.Audio:0", src=Multiply_0.Out}
   {dst="@UE.OutputFormat.Stereo.Audio:1", src=Multiply_1.Out}
  ]
 }]
properties {
 assetImportData=&self:AssetImportData
 bOverrideConcurrency=true
 numChannels=2
 outputFormat=Stereo
 sampleRate=48000
 soundClassObject=&Fort_Music_NPC
 concurrencyOverrides {maxCount=1, resolutionRule=PreventNew}
}
